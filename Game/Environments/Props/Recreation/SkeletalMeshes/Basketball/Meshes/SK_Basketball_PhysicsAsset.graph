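# Physics asset for the SK_Basketball skeletal mesh, a recreation prop, giving the ball its collision shape for physics.
# It holds a single body, a sphere sized like a ball, bound to the Melee_TwoHanded_Handle bone.
# There are no constraints, so the asset describes one rigid shape with no jointed parts.
# The bone name suggests the mesh uses a two-handed melee weapon skeleton.
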PhysicsAsset /Game/Environments/Props/Recreation/SkeletalMeshes/Basketball/Meshes/SK_Basketball_PhysicsAsset
bodies {Melee_TwoHanded_Handle [{shapeType=Sphere, radius=11.865}]}
constraints []
